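# Physics asset that supplies collision shapes for the cloth asset of the M_MED_Gimmick male player body.
# Capsules cover the pelvis, lower spine and four-bone left and right dyn_belt chains, so belt collision follows those bones.
# A wide box on the root bone forms a ground slab at the character's feet.
# Leg, foot and forearm bodies exist but hold no shapes, and there are no constraints.
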PhysicsAsset /BRCosmetics/Characters/Player/Male/Medium/Bodies/M_MED_Gimmick/Meshes/ClothAsset/M_MED_Gimmick_ClothAssetCollision
bodies {
 calf_l []
 calf_r []
 dyn_belt_l_01 [{shapeType=Capsule, length=11.8293, radius=8, shapeName=dyn_belt_r_01_capsule, center=<11.5402 -0.0009 7.3452>, rotation=<rot 90 0 0>}]
 dyn_belt_l_02 [{
   shapeType=Capsule
   length=16.0763
   radius=3.753
   shapeName=dyn_belt_r_01_capsule
   center=<3.9525 -0.8863 2.6731>
   rotation=<rot 87.5124 179.9997 179.9997>
  }]
 dyn_belt_l_03 [{shapeType=Capsule, length=10.0176, radius=6.2536, shapeName=dyn_belt_r_01_capsule, center=<4.0725 -3.172 4.211>, rotation=<rot 90 0 0>}]
 dyn_belt_l_04 [{shapeType=Capsule, length=3.1872, radius=7.3613, shapeName=dyn_belt_r_01_capsule, center=<3.7635 -4.7961 3.8284>, rotation=<rot 90 0 0>}]
 dyn_belt_r_01 [{shapeType=Capsule, length=11.8293, radius=8, shapeName=dyn_belt_r_01_capsule, center=<11.5402 -0.0009 7.3452>, rotation=<rot 90 0 0>}]
 dyn_belt_r_02 [{
   shapeType=Capsule
   length=16.0763
   radius=3.753
   shapeName=dyn_belt_r_01_capsule
   center=<3.9881 1.6825 2.2367>
   rotation=<rot 87.5124 179.9997 179.9997>
  }]
 dyn_belt_r_03 [{shapeType=Capsule, length=10.5474, radius=5.7238, shapeName=dyn_belt_r_01_capsule, center=<4.0157 3.1244 3.7294>, rotation=<rot 90 0 0>}]
 dyn_belt_r_04 [{
   shapeType=Capsule
   length=3.1872
   radius=7.3613
   shapeName=dyn_belt_r_01_capsule
   center=<3.5632 5.4956 4.3109>
   rotation=<rot 47.8093 93.7481 92.7788>
  }]
 foot_l []
 foot_r []
 lowerarm_l []
 lowerarm_r []
 pelvis [{shapeType=Capsule, length=7.9865, radius=11.2507, center=<-2.5281 -2.1225 0.0001>}]
 root [{shapeType=Box, extentX=327.5, extentY=327.5, extentZ=62.3399, shapeName=root_box, center=<0 0 -29.8853>}]
 spine_01 [{shapeType=Capsule, length=7.9663, radius=10.1267, center=<0.3095 -0.6696 0>}]
 spine_02 [{shapeType=Capsule, length=8.2792, radius=9.2496, center=<-1.0816 -0.436 -0.0002>}]
 thigh_l []
 thigh_r []
}
constraints []
