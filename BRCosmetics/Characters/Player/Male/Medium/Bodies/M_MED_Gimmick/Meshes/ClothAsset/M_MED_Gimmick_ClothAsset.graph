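# Chaos Cloth asset that simulates cloth on the M_MED_Gimmick player body cosmetic, built by an embedded Dataflow graph.
# It merges sim and render mesh imports, adds a separate backstop accessory mesh, then binds them with a proxy deformer and smoothed skinning blend.
# Weight maps drive max distance, aerodynamics, anim drive and backstop radius.
# Hand-picked fixed ends anchor geodesic tethers, with capsule collision and clamped root-bone-space velocities feeding one LOD.
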ChaosClothAsset /BRCosmetics/Characters/Player/Male/Medium/Bodies/M_MED_Gimmick/Meshes/ClothAsset/M_MED_Gimmick_ClothAsset
name=EmbeddedDataflow
graph=&self:EmbeddedDataflow
connections [
 {
  fromNode=ReferenceBone
  fromPinGuid=#1
  toNode=ClothAssetTerminal
  toPin="CollectionLods[0]"
  toPinGuid=#2
 }
 {
  fromNode=SimAccessoryMeshNode
  fromPinGuid=#3
  toNode=ProceduralSelection Render
  toPinGuid=#4
 }
 {
  fromNode=AnimDrive WeightMap_0
  fromPin=OutputName.StringValue
  fromPinGuid=#5
  toNode=SimulationAnimDriveConfig_0
  toPin=AnimDriveStiffness.WeightMap
  toPinGuid=#6
 }
 {
  fromNode=ProceduralSelection Render
  fromPinGuid=#7
  toNode=ProceduralSelection SIM
  toPinGuid=#8
 }
 {
  fromNode=SimulationAnimDriveConfig_0
  fromPinGuid=#9
  toNode=Backstop Radius WeightMap
  toPinGuid=#10
 }
 {
  fromNode=SkeletalMeshImport_Render Mesh_0
  fromPinGuid=#11
  toNode=MergeClothCollections
  toPin="Collections[1]"
  toPinGuid=#12
 }
 {
  fromNode=ProceduralSelection Render
  fromPin=OutputName
  fromPinGuid=#13
  toNode=ProxyDeformer_0
  toPin="SelectionFilterSets[0].RenderSelection.StringValue"
  toPinGuid=#14
 }
 {
  fromNode=LiftDrag WeightMap_0
  fromPinGuid=#15
  toNode=SimulationAerodynamicsConfig_0
  toPinGuid=#16
 }
 {
  fromNode=SimulationSolverConfig
  fromPinGuid=#17
  toNode=ReferenceBone
  toPinGuid=#18
 }
 {
  fromNode=ProceduralSelection SIM
  fromPinGuid=#19
  toNode=ProxyDeformer_0
  toPinGuid=#20
 }
 {
  fromNode=SimulationDampingConfig
  fromPinGuid=#21
  toNode=LiftDrag WeightMap_0
  toPinGuid=#22
 }
 {
  fromNode=LiftDrag WeightMap_0
  fromPin=OutputName.StringValue
  fromPinGuid=#23
  toNode=SimulationAerodynamicsConfig_0
  toPin=Lift.WeightMap
  toPinGuid=#24
 }
 {
  fromNode=SkeletalMeshImport_Accessory Mesh
  fromPinGuid=#25
  toNode=SimAccessoryMeshNode
  toPin=SimAccessoryMeshCollection
  toPinGuid=#26
 }
 {
  fromNode=LiftDrag WeightMap_0
  fromPin=OutputName.StringValue
  fromPinGuid=#23
  toNode=SimulationAerodynamicsConfig_0
  toPin=Drag.WeightMap
  toPinGuid=#27
 }
 {
  fromNode=SimulationLongRangeAttachmentConfig_0
  fromPinGuid=#28
  toNode=SimulationBendingConfig
  toPinGuid=#29
 }
 {
  fromNode=SimulationStretchConfig
  fromPinGuid=#30
  toNode=SimulationDampingConfig
  toPinGuid=#31
 }
 {fromNode=WeightMap_0, fromPinGuid=#32, toNode=SimulationMassConfig, toPinGuid=#33}
 {
  fromNode=SimulationMaxDistanceConfig
  fromPin=KinematicVertices3D
  fromPinGuid=#34
  toNode=SkinningBlend_0
  toPin=KinematicVertices3D.StringValue
  toPinGuid=#35
 }
 {
  fromNode=Selection
  fromPinGuid=#36
  toNode=SimulationLongRangeAttachmentConfig_0
  toPinGuid=#37
 }
 {
  fromNode=SimulationBendingConfig
  fromPinGuid=#38
  toNode=SimulationStretchConfig
  toPinGuid=#39
 }
 {
  fromNode=ProceduralSelection SIM
  fromPin=OutputName
  fromPinGuid=#40
  toNode=ProxyDeformer_0
  toPin="SelectionFilterSets[0].SimSelection.StringValue"
  toPinGuid=#41
 }
 {
  fromNode=SimulationVelocityScaleConfig_0
  fromPinGuid=#42
  toNode=AnimDrive WeightMap_0
  toPinGuid=#43
 }
 {
  fromNode=AnimDrive WeightMap_0
  fromPinGuid=#44
  toNode=SimulationAnimDriveConfig_0
  toPinGuid=#45
 }
 {
  fromNode=SimulationMaxDistanceConfig
  fromPinGuid=#46
  toNode=SkinningBlend_0
  toPinGuid=#47
 }
 {
  fromNode=Selection
  fromPin=OutputName.StringValue
  fromPinGuid=#48
  toNode=SimulationLongRangeAttachmentConfig_0
  toPin=FixedEndSet.StringValue
  toPinGuid=#49
 }
 {
  fromNode=SimulationCollisionConfig_1
  fromPinGuid=#50
  toNode=SimulationSolverConfig
  toPinGuid=#51
 }
 {
  fromNode=SimulationGravityConfig_0
  fromPinGuid=#52
  toNode=SimulationVelocityScaleConfig_0
  toPinGuid=#53
 }
 {
  fromNode=SimulationAerodynamicsConfig_0
  fromPinGuid=#54
  toNode=SimulationGravityConfig_0
  toPinGuid=#55
 }
 {
  fromNode=MergeClothCollections
  fromPinGuid=#56
  toNode=SimAccessoryMeshNode
  toPinGuid=#57
 }
 {
  fromNode=MaxDist WeightMap_0
  fromPinGuid=#58
  toNode=SimulationMaxDistanceConfig
  toPinGuid=#59
 }
 {
  fromNode=MaxDist WeightMap_0
  fromPin=OutputName.StringValue
  fromPinGuid=#60
  toNode=SimulationMaxDistanceConfig
  toPin=MaxDistance.WeightMap
  toPinGuid=#61
 }
 {fromNode=SkinningBlend_0, fromPinGuid=#62, toNode=WeightMap_0, toPinGuid=#63}
 {
  fromNode=SimulationBackstopConfig_0
  fromPinGuid=#64
  toNode=SetPhysicsAsset - Capsules_0
  toPinGuid=#65
 }
 {
  fromNode=SetPhysicsAsset - Capsules_0
  fromPinGuid=#66
  toNode=SimulationCollisionConfig_1
  toPinGuid=#67
 }
 {fromNode=SimulationMassConfig, fromPinGuid=#68, toNode=Selection, toPinGuid=#69}
 {
  fromNode=SkinningBlend_0
  fromPin=SkinningBlendName
  fromPinGuid=#70
  toNode=WeightMap_0
  toPin=InputName.StringValue
  toPinGuid=#71
 }
 {
  fromNode=SkeletalMesh_1
  fromPin=SkeletalMesh
  fromPinGuid=#72
  toNode=SkeletalMeshImport_Render Mesh_0
  toPin=SkeletalMesh
  toPinGuid=#73
 }
 {
  fromNode=SkeletalMeshImport_Sim Mesh_0
  fromPinGuid=#74
  toNode=MergeClothCollections
  toPin="Collections[0]"
  toPinGuid=#75
 }
 {
  fromNode=Backstop Radius WeightMap
  fromPin=OutputName.StringValue
  fromPinGuid=#76
  toNode=SimulationBackstopConfig_0
  toPin=BackstopRadius.WeightMap
  toPinGuid=#77
 }
 {
  fromNode=Backstop Radius WeightMap
  fromPinGuid=#78
  toNode=SimulationBackstopConfig_0
  toPinGuid=#79
 }
 {
  fromNode=ProxyDeformer_0
  fromPinGuid=#80
  toNode=MaxDist WeightMap_0
  toPinGuid=#81
 }
]
nodes [
 ClothAssetTerminal {type=FChaosClothAssetTerminalNode_v2, guid=#82}
 SimulationMaxDistanceConfig {
  type=FChaosClothAssetSimulationMaxDistanceConfigNode
  guid=#83
  properties {MaxDistance=(bIsAnimatable=True,High=200,WeightMap="MaxDistance")}
 }
 SimulationSolverConfig {
  type=FChaosClothAssetSimulationSolverConfigNode
  guid=#84
  properties {MaxNumIterations=2, NumIterations=1, NumSubstepsImported=(ImportedValue=1)}
 }
 SkeletalMesh_1 {
  type=FGetSkeletalMeshDataflowNode
  guid=#85
  inputPins []
  outputPins [SkeletalMesh {type="TObjectPtr<USkeletalMesh>"}]
  properties {
   PropertyName=SkeletalMesh
   SkeletalMesh=&SkeletalMesh'&/M_MED_Gimmick_ClothEditor'
  }
 }
 SkeletalMeshImport_Sim Mesh_0 {
  type=FChaosClothAssetSkeletalMeshImportNode_v2
  guid=#86
  properties {
   SkeletalMesh=&SkeletalMesh'&/M_MED_Gimmick_SIM'
   UVScale=(X=120,Y=120)
   bImportRenderMesh=False
   bImportSimMesh=True
   bImportSingleSection=False
  }
 }
 SkeletalMeshImport_Render Mesh_0 {
  type=FChaosClothAssetSkeletalMeshImportNode_v2
  guid=#87
  properties {SectionIndex=1, SkeletalMesh=None, bImportRenderMesh=True, bImportSimMesh=False, bImportSingleSection=True}
 }
 MergeClothCollections {type=FChaosClothAssetMergeClothCollectionsNode_v2, guid=#88}
 MaxDist WeightMap_0 {type=FChaosClothAssetWeightMapNode, guid=#89, properties {OutputName=(StringValue="MaxDistance")}}
 SimulationLongRangeAttachmentConfig_0 {
  type=FChaosClothAssetSimulationLongRangeAttachmentConfigNode_v2
  guid=#90
  inputPins [
   Collection {type=FManagedArrayCollection}
   FixedEndSet.StringValue {type=FString}
   TetherStiffness.WeightMap {type=FString}
   TetherScale.WeightMap {type=FString}
   "CustomTetherData[0].CustomFixedEndSet.StringValue" {type=FString}
   "CustomTetherData[0].CustomDynamicEndSet.StringValue" {type=FString}
  ]
  outputPins [Collection {type=FManagedArrayCollection}]
  properties {
   CustomTetherData=(())
   FixedEndSet=(StringValue="KinematicVertices3D")
   TetherScale=(bIsAnimatable=True,Low=1,High=1,WeightMap="TetherScale")
   TetherStiffness=(bIsAnimatable=True,Low=1,High=1,WeightMap="TetherStiffness")
   bEnableCustomTetherGeneration=False
   bUseGeodesicTethers=True
   bWarnDuplicateProperty=True
  }
 }
 SimulationBackstopConfig_0 {
  type=FChaosClothAssetSimulationBackstopConfigNode
  guid=#91
  inputPins [
   Collection {type=FManagedArrayCollection}
   BackstopDistance.WeightMap {type=FString}
   BackstopRadius.WeightMap {type=FString}
   BackstopMeshName.StringValue {type=FString}
  ]
  outputPins [Collection {type=FManagedArrayCollection}]
  properties {
   BackstopDistance=(bIsAnimatable=True,WeightMap="BackstopDistance")
   BackstopMeshName=(StringValue="AccessoryMesh")
   BackstopRadius=(bIsAnimatable=True,Low=50,High=50,WeightMap="BackstopRadius")
   bWarnDuplicateProperty=True
  }
 }
 SimulationAnimDriveConfig_0 {
  type=FChaosClothAssetSimulationAnimDriveConfigNode
  guid=#92
  inputPins [
   Collection {type=FManagedArrayCollection}
   AnimDriveStiffness.WeightMap {type=FString}
   AnimDriveDamping.WeightMap {type=FString}
  ]
  outputPins [Collection {type=FManagedArrayCollection}]
  properties {
   AnimDriveDamping=(bIsAnimatable=True,High=1,WeightMap="AnimDriveDamping")
   AnimDriveStiffness=(bIsAnimatable=True,High=1,WeightMap="AnimDriveStiffness")
   bWarnDuplicateProperty=True
  }
 }
 AnimDrive WeightMap_0 {type=FChaosClothAssetWeightMapNode, guid=#93, properties {OutputName=(StringValue="AnimDrive")}}
 SimulationAerodynamicsConfig_0 {
  type=FChaosClothAssetSimulationAerodynamicsConfigNode
  guid=#94
  properties {
   Drag=(bIsAnimatable=True,Low=0.1,High=0.7,WeightMap="Drag")
   Lift=(bIsAnimatable=True,Low=0.1,High=0.4,WeightMap="Lift")
   TurbulenceRatio=1
   WindVelocity=(X=0,Y=-1.3,Z=-1.5)
   WindVelocitySpace=ReferenceBoneSpace
  }
 }
 LiftDrag WeightMap_0 {type=FChaosClothAssetWeightMapNode, guid=#95, properties {OutputName=(StringValue="LiftDrag")}}
 SimulationBendingConfig {
  type=FChaosClothAssetSimulationBendingConfigNode
  guid=#96
  inputPins [
   Collection {type=FManagedArrayCollection}
   FlatnessRatio.WeightMap {type=FString}
   RestAngle.WeightMap {type=FString}
   BendingStiffness.WeightMap {type=FString}
   BendingStiffnessWarp.WeightMap {type=FString}
   BendingStiffnessWeft.WeightMap {type=FString}
   BendingStiffnessBias.WeightMap {type=FString}
   BendingDamping.WeightMap {type=FString}
   BendingAnisoDamping.WeightMap {type=FString}
   AnisoBucklingRatio.WeightMap {type=FString}
   BucklingStiffness.WeightMap {type=FString}
   BucklingStiffnessWarp.WeightMap {type=FString}
   BucklingStiffnessWeft.WeightMap {type=FString}
   BucklingStiffnessBias.WeightMap {type=FString}
   BucklingRatioWeighted.WeightMap {type=FString}
  ]
  properties {
   AnisoBucklingRatio=(bIsAnimatable=True,Low=0.5,High=0.5,WeightMap="BucklingRatio")
   BendingAnisoDamping=(bIsAnimatable=True,Low=3,High=5,WeightMap="BendingAnisoDamping")
   BendingStiffnessBias=(bIsAnimatable=True,Low=5,High=5,WeightMap="BendingStiffnessBias")
   BendingStiffnessWarp=(bIsAnimatable=True,Low=5,High=5,WeightMap="BendingStiffnessWarp")
   BendingStiffnessWeft=(bIsAnimatable=True,Low=5,High=5,WeightMap="BendingStiffnessWeft")
   BucklingStiffnessBias=(bIsAnimatable=True,Low=2,High=50,WeightMap="BucklingStiffnessBias")
   BucklingStiffnessWarp=(bIsAnimatable=True,Low=2,High=50,WeightMap="BucklingStiffnessWarp")
   BucklingStiffnessWeft=(bIsAnimatable=True,Low=2,High=50,WeightMap="BucklingStiffnessWeft")
  }
 }
 SimulationStretchConfig {
  type=FChaosClothAssetSimulationStretchConfigNode
  guid=#97
  properties {
   StretchAnisoDamping=(bIsAnimatable=True,Low=2,High=1,WeightMap="StretchAnisoDamping")
   StretchStiffnessBias=(bIsAnimatable=True,Low=10,High=1000000,WeightMap="StretchStiffnessBias")
   StretchStiffnessWarp=(bIsAnimatable=True,Low=100000,High=1000000,WeightMap="StretchStiffnessWarp")
   StretchStiffnessWeft=(bIsAnimatable=True,Low=100000,High=1000000,WeightMap="StretchStiffnessWeft")
  }
 }
 SimulationMassConfig {
  type=FChaosClothAssetSimulationMassConfigNode
  guid=#98
  properties {DensityWeighted=(Low=1,High=1,WeightMap="Density")}
 }
 SimulationDampingConfig {
  type=FChaosClothAssetSimulationDampingConfigNode
  guid=#99
  properties {
   DampingCoefficientWeighted=(WeightMap="DampingCoefficient")
   LocalDampingAngularCoefficient=0.2
   LocalDampingLinearCoefficient=0.2
   LocalDampingSpace=ReferenceBoneSpace
  }
 }
 SimulationGravityConfig_0 {
  type=FChaosClothAssetSimulationGravityConfigNode
  guid=#100
  properties {GravityOverrideImported=(ImportedValue=(X=0,Y=0,Z=-980.665))}
 }
 SimulationVelocityScaleConfig_0 {
  type=FChaosClothAssetSimulationVelocityScaleConfigNode
  guid=#101
  inputPins [Collection {type=FManagedArrayCollection}]
  outputPins [Collection {type=FManagedArrayCollection}]
  properties {
   AngularVelocityScale=0.75
   FictitiousAngularScale=0.5
   FictitiousForcesModel=FullScale
   LinearVelocityScale=(X=0.75,Y=0.75,Z=0.75)
   MaxAngularAcceleration=3
   MaxAngularVelocity=2
   MaxLinearAcceleration=(X=1500,Y=1000,Z=1500)
   MaxLinearVelocity=(X=700,Y=700,Z=700)
   MaxVelocityScale=2
   VelocityScaleSpace=ReferenceBoneSpace
   bEnableAngularAccelerationClamping=True
   bEnableAngularVelocityClamping=True
   bEnableLinearAccelerationClamping=True
   bEnableLinearVelocityClamping=True
   bWarnDuplicateProperty=True
  }
 }
 SimulationCollisionConfig_1 {
  type=FChaosClothAssetSimulationCollisionConfigNode
  guid=#102
  properties {
   ClothCollisionThicknessImported=(bIsAnimatable=True,WeightMap="ClothCollisionThickness")
   CollisionThicknessImported=(ImportedValue=0.5)
   FrictionCoefficientWeighted=(bIsAnimatable=True,Low=0.2,High=0.2,WeightMap="FrictionCoefficient")
   InnerCollisionThickness=(WeightMap="InnerCollisionThickness")
   bUseCCD=False
  }
 }
 SetPhysicsAsset - Capsules_0 {
  type=FChaosClothAssetSetPhysicsAssetNode
  guid=#103
  properties {PhysicsAsset=/Script/Engine.PhysicsAsset'&/M_MED_Gimmick_ClothAssetCollision'}
 }
 WeightMap_0 {
  type=FChaosClothAssetWeightMapNode
  guid=#104
  properties {MapOverrideType=Add, MeshTarget=Render, OutputName=()}
 }
 ProxyDeformer_0 {
  type=FChaosClothAssetProxyDeformerNode_v3
  guid=#105
  properties {InfluenceRadius=5, SelectionFilterSets=(()), bPreserveRenderTangents=True}
 }
 ReferenceBone {
  type=FChaosClothAssetReferenceBoneNode
  guid=#106
  inputPins [Collection {type=FManagedArrayCollection}]
  outputPins [Collection {type=FManagedArrayCollection}]
  properties {CalculateDefaultReferenceBone=(), ReferenceBone=(Name="root")}
 }
 SkinningBlend_0 {
  type=FChaosClothAssetSkinningBlendNode
  guid=#107
  inputPins [
   Collection {type=FManagedArrayCollection}
   KinematicVertices3D.StringValue {type=FString}
  ]
  outputPins [
   Collection {type=FManagedArrayCollection}
   SkinningBlendName {type=FString}
  ]
  properties {KinematicVertices3D=(), SkinningBlendName=RenderDeformerSkinningBlend, bUseSmoothTransition=True}
 }
 SimAccessoryMeshNode {
  type=FChaosClothAssetSimAccessoryMeshNode
  guid=#108
  inputPins [
   Collection {type=FManagedArrayCollection}
   SimAccessoryMeshCollection {type=FManagedArrayCollection}
  ]
  outputPins [
   Collection {type=FManagedArrayCollection}
   AccessoryMeshName {type=FString}
  ]
  properties {AccessoryMeshName=AccessoryMesh, bUseSimImportVertexID=True}
 }
 SkeletalMeshImport_Accessory Mesh {
  type=FChaosClothAssetSkeletalMeshImportNode_v2
  guid=#109
  properties {
   SkeletalMesh=&SkeletalMesh'&/M_MED_Gimmick_BACKSTOP'
   UVScale=(X=120,Y=120)
   bImportRenderMesh=False
   bImportSimMesh=True
   bImportSingleSection=False
  }
 }
 Backstop Radius WeightMap {type=FChaosClothAssetWeightMapNode, guid=#110, properties {OutputName=()}}
 Selection {
  type=FChaosClothAssetSelectionNode_v2
  guid=#111
  inputPins [
   Collection {type=FManagedArrayCollection}
   InputName.StringValue {type=FString}
   TransferCollection {type=FManagedArrayCollection}
  ]
  outputPins [
   Collection {type=FManagedArrayCollection}
   OutputName.StringValue {type=FString}
  ]
  properties {
   Group=(Name="SimVertices3D")
   Import=()
   ImportSecondary=()
   Indices=(215,216,217,218,219,220,221,226,227,228,232,238,239)
   InputName=()
   OutputName=(StringValue="LRA_FixedEnds")
   RemoveIndices=()
   SelectionOverrideType=ReplaceAll
   SimTransferType=Use2DSimMesh
   Transfer=()
   TransferSelectionThreshold=0.95
  }
 }
 ProceduralSelection Render {
  type=FChaosClothAssetProceduralSelectionNode
  guid=#112
  inputPins [
   Collection {type=FManagedArrayCollection}
   ConversionInputName.StringValue {type=FString}
  ]
  outputPins [
   Collection {type=FManagedArrayCollection}
   OutputName {type=FString}
  ]
  properties {ConversionInputName=(), Group=(Name="RenderVertices"), OutputName=RenderSelectionProcedural, SelectionType=SelectAll}
 }
 ProceduralSelection SIM {
  type=FChaosClothAssetProceduralSelectionNode
  guid=#113
  inputPins [
   Collection {type=FManagedArrayCollection}
   ConversionInputName.StringValue {type=FString}
  ]
  outputPins [
   Collection {type=FManagedArrayCollection}
   OutputName {type=FString}
  ]
  properties {ConversionInputName=(), Group=(Name="SimVertices3D"), OutputName=SimSelectionProcedural, SelectionType=SelectAll}
 }
]
